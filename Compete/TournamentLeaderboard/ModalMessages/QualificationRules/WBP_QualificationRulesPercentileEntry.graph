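# List row widget for the tournament leaderboard's qualification rules, showing a top-percentile label and its points over a tile background.
# SetPointsNeededForPercentile writes non-negative points, then always reports itself and its validity to ParentQualificationRules via OnEntryFinishedSetup.
# OnListItemObjectSet treats the list item as another instance of this same widget class and copies both its texts.
# PreConstruct fills placeholder percentage and random points at design time only.
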WidgetBlueprint /Compete/TournamentLeaderboard/ModalMessages/QualificationRules/WBP_QualificationRulesPercentileEntry
graphs [
 EventGraph {nodes [
   n1 = Event PreConstruct()
   > Parent: PreConstruct(IsDesignTime=@n1.IsDesignTime)
   > Branch(Condition=@n1.IsDesignTime)
   > SetText!(
    execute
    self:WBP UIKit Block Text Object Reference=Get BlockText_TopPercentage() -> (BlockText_TopPercentage:WBP UIKit Block Text Object Reference)
    Text:Text="LOCTEXT(\"Top 5%\")"
   ) -> (then)
   > SetText(
    self=Get BlockText_Points() -> (BlockText_Points:WBP UIKit Block Text Object Reference)
    Text="To Text (Integer)"(Value=RandomInteger!(Max:int=1001) -> (ReturnValue:int=0))
   )
   n2 = Event OnListItemObjectSet()
   > n3 = Cast To WBP_QualificationRulesPercentileEntry!(execute, Object=@n2.ListItemObject) -> (then, CastFailed, AsWBP Qualification Rules Percentile Entry:WBP Qualification Rules Percentile Entry Object Reference) {is=title K2Node_DynamicCast}
   n4 = Get BlockText_TopPercentage!(self:WBP Qualification Rules Percentile Entry Object Reference=@n3) -> (BlockText_TopPercentage)
   > SetText(
    self=Get BlockText_TopPercentage!() -> (BlockText_TopPercentage)
    Text="GetText (Text)"!(self:Text Object Reference=Get TextBlock!(self:WBP UIKit Block Text Object Reference=@n4) -> (TextBlock:Common Text Object Reference)) -> (ReturnValue:Text)
   )
   n5 = Get BlockText_Points!(self:WBP Qualification Rules Percentile Entry Object Reference=@n3) -> (BlockText_Points)
   > SetText(self=Get BlockText_Points!() -> (BlockText_Points), Text="GetText (Text)"(self=Get TextBlock(self=@n5)))
  ]}
 SetPointsNeededForPercentile {nodes [
   n1 = SetPointsNeededForPercentile!() -> (then, InPoints:int) {is=title K2Node_FunctionEntry}
   n2 = "integer >= integer"(A=@n1)
   > n3 = Branch(Condition=@n2)
   > n4 = SetPointsText!(execute, self, InText:Text="To Text (Integer)"(Value=@n1)) -> (then)
   OnEntryFinishedSetup!(
    execute=[@n4.then, @n3.else]
    self:WBP Qualification Rules Percentiles Object Reference=Get ParentQualificationRules() -> (ParentQualificationRules:WBP Qualification Rules Percentiles Object Reference)
    EntryWidget:WBP Qualification Rules Percentile Entry Object Reference=Self-Reference()
    bValidEntry:bool=@n2
   ) -> (then)
  ]}
 SetPointsText {nodes [
   n1 = SetPointsText!() -> (then, InText:Text) {is=title K2Node_FunctionEntry}
   > SetText(self=Get BlockText_Points(), Text=@n1)
  ]}
 SetTopPercentileText {nodes [
   n1 = SetTopPercentileText!() -> (then, InText:Text) {is=title K2Node_FunctionEntry}
   > SetText(self=Get BlockText_TopPercentage(), Text=@n1)
  ]}
]
variables [
 ParentListView {type=CommonListView, default=null}
 ParentQualificationRules {type=WBP_QualificationRulesPercentiles_C, default=null}
]
widgets {
 info {inheritedWidgetCount=0, parentClass=&CommonUserWidget, rootWidgetClass=&GridPanel, widgetCount=5}
 widgets [
  GridPanel_29 {widgetClassPath=&GridPanel, widget=GridPanel_29, properties {flowDirectionPreference=Culture, columnFill [1], rowFill [1]}}
  Block_Background {
   widgetClassPath=/FortUILibrary/Systems/ModularButton/Blocks/WBP_UIKit_Block_BackgroundGeneric
   bIsVariable=true
   parent=GridPanel_29
   widget=Block_Background
   properties {backgroundMaterial=/Compete/MaterialInstances/MI_UI_CompeteLeaderboardTileBackground}
   slot {class=GridSlot, properties {columnSpan=2}}
  }
  HorizontalBox_0 {
   widgetClassPath=&HorizontalBox
   parent=GridPanel_29
   widget=HorizontalBox_0
   slot {class=GridSlot, properties {layer=1, verticalAlignment=VAlign_Center, padding=<ltrb 30 15 30 15>}}
  }
  BlockText_TopPercentage {
   widgetClassPath=/FortUILibrary/Systems/ModularButton/Blocks/WBP_UIKit_Block_Text
   bIsVariable=true
   parent=HorizontalBox_0
   widget=BlockText_TopPercentage
   slot {class=HorizontalBoxSlot, properties {horizontalAlignment=HAlign_Left}}
  }
  BlockText_Points {
   widgetClassPath=&WBP_UIKit_Block_Text
   bIsVariable=true
   parent=HorizontalBox_0
   widget=BlockText_Points
   slot {class=HorizontalBoxSlot, properties {horizontalAlignment=HAlign_Right, verticalAlignment=VAlign_Center, size {sizeRule=Fill, value=1}}}
  }
 ]
}
