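# Tournament qualification-rules widget listing percentile payout thresholds under leaderboard "Top %" and Points headers.
# SetPercentileThresholds clears the list and, if the feature is enabled, creates a pending entry widget per threshold, each receiving an async FortCompeteUIManager points query callback.
# OnEntryFinishedSetup removes entries from pending, adds valid ones to the list and fires OnAddedValidEntry.
# The created entry widgets themselves serve as the list view's item objects.
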WidgetBlueprint /Compete/TournamentLeaderboard/ModalMessages/QualificationRules/WBP_QualificationRulesPercentiles
graphs [
 EventGraph {links [], nodes []}
 OnAddedValidEntry {nodes [OnAddedValidEntry() {is=title K2Node_FunctionEntry}]}
 OnEntryFinishedSetup {nodes [
   n1 = OnEntryFinishedSetup!() -> (then, EntryWidget:WBP Qualification Rules Percentile Entry Object Reference, bValidEntry:bool=~) {is=title K2Node_FunctionEntry}
   > Is Valid(InputObject=@n1)
   > Remove Item!(
    execute
    TargetArray:Array of WBP Qualification Rules Percentile Entry Object References=Get EntriesPendingQueries() -> (EntriesPendingQueries:Array of WBP Qualification Rules Percentile Entry Object References)
    Item:"WBP Qualification Rules Percentile Entry Object Reference (by ref)"=@n1
   ) -> (then, ReturnValue:bool)
   > Branch(Condition=@n1.bValidEntry)
   > AddItem!(
    execute
    self:List View Object Reference=Get EntryCommonListView() -> (EntryCommonListView:Common List View Object Reference)
    Item:Object Reference=@n1
   ) -> (then)
   > Call On Added Valid Entry!(execute:Exec, self:WBP Qualification Rules Percentiles Object Reference) -> (then:Exec) {is=title K2Node_CallDelegate}
  ]}
 SetPercentileThresholds {nodes [
   n1 = SetPercentileThresholds!() -> (
    then
    InThresholds:Array of Tournament Payout Threshold Data Structures
    TotalTournamentEntries:int
    LeaderboardCategory:Fort Compete UILeaderboard Category Structure
   ) {is=title K2Node_FunctionEntry}
   > Set List Items!(
    execute
    self:List View Object Reference=Get EntryCommonListView()
    InListItems:Array of Object References=Make Array!() -> (Array:Array of Object References)
   ) -> (then) {is=title K2Node_CallFunction}
   > n2 = IsQualificationRulesPercentileFeatureEnabled!(execute) -> (then, ReturnValue:bool)
   > Branch(Condition=AND Boolean(A=@n2, B=Is Not Empty!(TargetArray:Array of Tournament Payout Threshold Data Structures=@n1) -> (ReturnValue:bool)))
   > Set Text(
    Text=Format Text!(
     Format=Get TopPercentFormatText() -> (TopPercentFormatText:Text)
     Leaderboard:Text=Break Fort Compete UILeaderboard Category!(FortCompeteUILeaderboardCategory:"Fort Compete UILeaderboard Category Structure (by ref)"=@n1.LeaderboardCategory) -> (LeaderboardName:Text, GroupingKey:String, InstanceId:String) {is=title K2Node_BreakStruct}
    ) -> (Result)
    self=Get Rank() -> (Rank:Common Text Object Reference)
   )
   > n3 = For Each Loop!(Exec, Array:Array of Tournament Payout Threshold Data Structures=@n1) -> (
    LoopBody
    Array Element_ScoringType:EPayoutScoringType Enum
    Array Element_Threshold:double=~
    Array Element_ThresholdDisplayText:Text
    Array Element_ThresholdDisplayTextHeader:Text
    Array Element_SourceLeaderboardDisplayName:Text
    Array Element_PayoutItemsData:Array of Tournament Payout Item Data Structures
    Array Element_CosmeticPayoutItemsVM:Array of Fort Item VM Object References
    Array Index
    Completed
   )
   > n4 = Create Widget!(
    execute:Exec
    Class:User Widget Class Reference=&/WBP_QualificationRulesPercentileEntry.WBP_QualificationRulesPercentileEntry_C
    OwningPlayer:Player Controller Object Reference
   ) -> (then:Exec, ReturnValue:WBP Qualification Rules Percentile Entry Object Reference)
   > SetTopPercentileText!(execute, self:WBP Qualification Rules Percentile Entry Object Reference=@n4, InText:Text=@n3.Array Element_ThresholdDisplayText) -> (then)
   > Set ParentQualificationRules!(
    execute
    ParentQualificationRules:WBP Qualification Rules Percentiles Object Reference=Self-Reference()
    self:WBP Qualification Rules Percentile Entry Object Reference=@n4
   ) -> (then, Output_Get:WBP Qualification Rules Percentiles Object Reference)
   > Add(
    TargetArray:Array of WBP Qualification Rules Percentile Entry Object References=Get EntriesPendingQueries()
    NewItem:"WBP Qualification Rules Percentile Entry Object Reference (by ref)"=@n4
   ) {is=title K2Node_CallArrayFunction}
   > GetPointsNeededForPercentile!(
    execute
    self:Fort Compete UIManager Object Reference=Get FortCompeteUIManager() -> (ReturnValue:Fort Compete UIManager Object Reference) {is=title K2Node_GetSubsystem}
    Percentile:double=@n3.Array Element_Threshold
    NumEntries:int=Get TotalTournamentEntries() -> (TotalTournamentEntries:int=0)
    LeaderboardCategory:"Fort Compete UILeaderboard Category Structure (by ref)"=Get LeaderboardCategory() -> (LeaderboardCategory:Fort Compete UILeaderboard Category Structure)
    OnRequestCompleted:Delegate=Create Event(self=@n4)
   ) -> (then)
  ]}
]
variables [
 EntriesPendingQueries {type="array<WBP_QualificationRulesPercentileEntry_C>", default []}
 OnAddedValidEntry {}
 TopPercentFormatText {type=string, default="{Leaderboard} Top %"}
]
widgets {
 info {inheritedWidgetCount=0, parentClass=&CommonUserWidget, rootWidgetClass=&GridPanel, widgetCount=6}
 widgets [
  GridPanel_45 {widgetClassPath=&GridPanel, widget=GridPanel_45, properties {flowDirectionPreference=Culture, columnFill [1], rowFill [0,1]}}
  Headers {
   widgetClassPath=&HorizontalBox
   parent=GridPanel_45
   widget=Headers
   properties {bOverride_Cursor=true}
   slot {class=GridSlot, properties {layer=1, padding=<ltrb 30 0 30 0>}}
  }
  Rank {
   widgetClassPath=&CommonTextBlock
   bIsVariable=true
   parent=Headers
   widget=Rank
   properties {
    minDesiredWidth=103
    mobileFontSizeMultiplier=1.14
    text="{Leaderboard} Top %"
    colorAndOpacity {colorUseRule=UseColor_Specified, specifiedColor=<rgba 0.5776 0.5776 0.5776 1>}
    font {
     bForceMonospaced=false
     bMaterialIsStencil=false
     fontMaterial=null
     fontObject=&HeadingNow
     letterSpacing=0
     monospacedWidth=1
     size=15.75
     skewAmount=0
     typefaceFontName=75 Medium
     outlineSettings {
      bApplyOutlineToDropShadows=false
      bMiteredCorners=false
      bSeparateFillAlpha=false
      outlineMaterial=null
      outlineSize=0
      outlineColor=<rgba 0 0 0 1>
     }
    }
    shadowOffset=<0 0>
   }
   slot {class=HorizontalBoxSlot, properties {horizontalAlignment=HAlign_Left}}
  }
  Points {
   widgetClassPath=&CommonTextBlock
   parent=Headers
   widget=Points
   properties {
    mobileFontSizeMultiplier=1.14
    text=Points
    colorAndOpacity {colorUseRule=UseColor_Specified, specifiedColor=<rgba 0.5776 0.5776 0.5776 1>}
    font {
     bForceMonospaced=false
     bMaterialIsStencil=false
     fontMaterial=null
     fontObject=&HeadingNow
     letterSpacing=0
     monospacedWidth=1
     size=15.75
     skewAmount=0
     typefaceFontName=75 Medium
     outlineSettings {
      bApplyOutlineToDropShadows=false
      bMiteredCorners=false
      bSeparateFillAlpha=false
      outlineMaterial=null
      outlineSize=0
      outlineColor=<rgba 0 0 0 1>
     }
    }
    shadowOffset=<0 0>
   }
   slot {class=HorizontalBoxSlot, properties {horizontalAlignment=HAlign_Right, size {sizeRule=Fill, value=1}}}
  }
  SizeBox_149 {widgetClassPath=&SizeBox, parent=GridPanel_45, widget=SizeBox_149, slot {class=GridSlot, properties {row=1}}}
  EntryCommonListView {
   widgetClassPath=&CommonListView
   bIsVariable=true
   parent=SizeBox_149
   widget=EntryCommonListView
   properties {
    bInEnableTouchAnimatedScrolling=true
    bReturnFocusToSelection=true
    entryWidgetClass=&WBP_QualificationRulesPercentileEntry_C
    numDesignerPreviewEntries=4
    scrollBarStyle {
     thickness=12
     draggedThumbImage {
      bIsDynamicallyLoaded=false
      drawAs=Box
      imageType=FullColor
      mirroring=NoMirror
      resourceName=../../../Engine/Content/Slate/Common/Scrollbar_Thumb.png
      resourceObject=null
      tiling=NoTile
      imageSize=<0 0>
      margin=<ltrb 0.25 0.25 0.25 0.25>
      outlineSettings {
       bUseBrushTransparency=false
       roundingType=HalfHeightRadius
       width=0
       color {colorUseRule=UseColor_Specified, specifiedColor=<rgba 0 0 0 0>}
       cornerRadii=<0 0 0 0>
      }
      tintColor {colorUseRule=UseColor_Specified, specifiedColor=<rgba 1 1 1 1>}
      uVRegion {bIsValid=false, max=<0 0>, min=<0 0>}
     }
     horizontalBackgroundImage {
      bIsDynamicallyLoaded=false
      drawAs=NoDrawType
      imageType=NoImage
      mirroring=NoMirror
      resourceName=None
      resourceObject=null
      tiling=NoTile
      imageSize=<0 0>
      margin=<ltrb 0 0 0 0>
      outlineSettings {
       bUseBrushTransparency=false
       roundingType=HalfHeightRadius
       width=0
       color {colorUseRule=UseColor_Specified, specifiedColor=<rgba 0 0 0 0>}
       cornerRadii=<0 0 0 0>
      }
      tintColor {colorUseRule=UseColor_Specified, specifiedColor=<rgba 1 1 1 1>}
      uVRegion {bIsValid=false, max=<0 0>, min=<0 0>}
     }
     horizontalBottomSlotImage {
      bIsDynamicallyLoaded=false
      drawAs=Image
      imageType=FullColor
      mirroring=NoMirror
      resourceName=../../../Engine/Content/Slate/Common/Scrollbar_Background_Horizontal.png
      resourceObject=null
      tiling=NoTile
      imageSize=<8 8>
      margin=<ltrb 0 0 0 0>
      outlineSettings {
       bUseBrushTransparency=false
       roundingType=HalfHeightRadius
       width=0
       color {colorUseRule=UseColor_Specified, specifiedColor=<rgba 0 0 0 0>}
       cornerRadii=<0 0 0 0>
      }
      tintColor {colorUseRule=UseColor_Specified, specifiedColor=<rgba 1 1 1 1>}
      uVRegion {bIsValid=false, max=<0 0>, min=<0 0>}
     }
     horizontalTopSlotImage {
      bIsDynamicallyLoaded=false
      drawAs=Image
      imageType=FullColor
      mirroring=NoMirror
      resourceName=../../../Engine/Content/Slate/Common/Scrollbar_Background_Horizontal.png
      resourceObject=null
      tiling=NoTile
      imageSize=<8 8>
      margin=<ltrb 0 0 0 0>
      outlineSettings {
       bUseBrushTransparency=false
       roundingType=HalfHeightRadius
       width=0
       color {colorUseRule=UseColor_Specified, specifiedColor=<rgba 0 0 0 0>}
       cornerRadii=<0 0 0 0>
      }
      tintColor {colorUseRule=UseColor_Specified, specifiedColor=<rgba 1 1 1 1>}
      uVRegion {bIsValid=false, max=<0 0>, min=<0 0>}
     }
     hoveredThumbImage {
      bIsDynamicallyLoaded=false
      drawAs=Box
      imageType=FullColor
      mirroring=NoMirror
      resourceName=../../../Engine/Content/Slate/Common/Scrollbar_Thumb.png
      resourceObject=null
      tiling=NoTile
      imageSize=<0 0>
      margin=<ltrb 0.25 0.25 0.25 0.25>
      outlineSettings {
       bUseBrushTransparency=false
       roundingType=HalfHeightRadius
       width=0
       color {colorUseRule=UseColor_Specified, specifiedColor=<rgba 0 0 0 0>}
       cornerRadii=<0 0 0 0>
      }
      tintColor {colorUseRule=UseColor_Specified, specifiedColor=<rgba 1 1 1 1>}
      uVRegion {bIsValid=false, max=<0 0>, min=<0 0>}
     }
     normalThumbImage {
      bIsDynamicallyLoaded=false
      drawAs=Box
      imageType=FullColor
      mirroring=NoMirror
      resourceName=../../../Engine/Content/Slate/Common/Scrollbar_Thumb.png
      resourceObject=null
      tiling=NoTile
      imageSize=<0 0>
      margin=<ltrb 0.25 0.25 0.25 0.25>
      outlineSettings {
       bUseBrushTransparency=false
       roundingType=HalfHeightRadius
       width=0
       color {colorUseRule=UseColor_Specified, specifiedColor=<rgba 0 0 0 0>}
       cornerRadii=<0 0 0 0>
      }
      tintColor {colorUseRule=UseColor_Specified, specifiedColor=<rgba 1 1 1 1>}
      uVRegion {bIsValid=false, max=<0 0>, min=<0 0>}
     }
     verticalBackgroundImage {
      bIsDynamicallyLoaded=false
      drawAs=NoDrawType
      imageType=NoImage
      mirroring=NoMirror
      resourceName=None
      resourceObject=null
      tiling=NoTile
      imageSize=<0 0>
      margin=<ltrb 0 0 0 0>
      outlineSettings {
       bUseBrushTransparency=false
       roundingType=HalfHeightRadius
       width=0
       color {colorUseRule=UseColor_Specified, specifiedColor=<rgba 0 0 0 0>}
       cornerRadii=<0 0 0 0>
      }
      tintColor {colorUseRule=UseColor_Specified, specifiedColor=<rgba 1 1 1 1>}
      uVRegion {bIsValid=false, max=<0 0>, min=<0 0>}
     }
     verticalBottomSlotImage {
      bIsDynamicallyLoaded=false
      drawAs=Image
      imageType=FullColor
      mirroring=NoMirror
      resourceName=../../../Engine/Content/Slate/Common/Scrollbar_Background_Vertical.png
      resourceObject=null
      tiling=NoTile
      imageSize=<8 8>
      margin=<ltrb 0 0 0 0>
      outlineSettings {
       bUseBrushTransparency=false
       roundingType=HalfHeightRadius
       width=0
       color {colorUseRule=UseColor_Specified, specifiedColor=<rgba 0 0 0 0>}
       cornerRadii=<0 0 0 0>
      }
      tintColor {colorUseRule=UseColor_Specified, specifiedColor=<rgba 1 1 1 1>}
      uVRegion {bIsValid=false, max=<0 0>, min=<0 0>}
     }
     verticalTopSlotImage {
      bIsDynamicallyLoaded=false
      drawAs=Image
      imageType=FullColor
      mirroring=NoMirror
      resourceName=../../../Engine/Content/Slate/Common/Scrollbar_Background_Vertical.png
      resourceObject=null
      tiling=NoTile
      imageSize=<8 8>
      margin=<ltrb 0 0 0 0>
      outlineSettings {
       bUseBrushTransparency=false
       roundingType=HalfHeightRadius
       width=0
       color {colorUseRule=UseColor_Specified, specifiedColor=<rgba 0 0 0 0>}
       cornerRadii=<0 0 0 0>
      }
      tintColor {colorUseRule=UseColor_Specified, specifiedColor=<rgba 1 1 1 1>}
      uVRegion {bIsValid=false, max=<0 0>, min=<0 0>}
     }
    }
   }
   slot {class=SizeBoxSlot}
  }
 ]
}
